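# Physics asset for the TangyRadishMagma Crust pickaxe mesh, defining collision and joint limits for the two-handed handle and four dynamic bow bones.
# A long capsule covers the handle, with short capsules on the left and right 00 and 01 bow bones.
# All four bow bones are constrained directly to the handle rather than chained to each other.
# Their swing is limited, differing between 00 and 01 bones, and twist is nearly locked.
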PhysicsAsset /BRCosmetics/Weapons/FORT_Melee/Pickaxe_TangyRadishMagma_Crust/Meshes/TangyRadishMagma_Crust_Axe_Physics
bodies {
 Melee_TwoHanded_Handle [{
   shapeType=Capsule
   length=49.7573
   radius=1.7237
   shapeName=Melee_TwoHanded_Handle_capsule
   center=<-0.0075 0.2479 23.2805>
   rotation=<rot 2.9223 -90 0.1547>
  }]
 dyn_bow_00_l [{shapeType=Capsule, length=2, radius=2.5, shapeName=dyn_bow_00_l_capsule, center=<3 0 0>, rotation=<rot 90 0 0>}]
 dyn_bow_00_r [{shapeType=Capsule, length=2, radius=2.5, shapeName=dyn_bow_00_l_capsule, center=<3 0 0>, rotation=<rot 90 0 0>}]
 dyn_bow_01_l [{shapeType=Capsule, length=2, radius=1, shapeName=dyn_bow_00_l_capsule, center=<2 0 0>, rotation=<rot 90 0 0>}]
 dyn_bow_01_r [{shapeType=Capsule, length=2, radius=1, shapeName=dyn_bow_00_l_capsule, center=<2 0 0>, rotation=<rot 90 0 0>}]
}
constraints [
 {bone1Name=dyn_bow_00_l, bone2Name=Melee_TwoHanded_Handle, swing1LimitDegrees=50, swing2LimitDegrees=15, twistLimitDegrees=1}
 {bone1Name=dyn_bow_00_r, bone2Name=Melee_TwoHanded_Handle, swing1LimitDegrees=50, swing2LimitDegrees=15, twistLimitDegrees=1}
 {bone1Name=dyn_bow_01_l, bone2Name=Melee_TwoHanded_Handle, swing1LimitDegrees=25, swing2LimitDegrees=25, twistLimitDegrees=1}
 {bone1Name=dyn_bow_01_r, bone2Name=Melee_TwoHanded_Handle, swing1LimitDegrees=25, swing2LimitDegrees=25, twistLimitDegrees=1}
]
